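# Additive unlit Niagara sprite material for the bubble FX on the Alien Agent character's head.
# MF_Sphere fakes a sphere normal and mask on each sprite, with the mask faded near the local X=0 plane and depth-faded.
# Two fake speculars, from the normal and a flipped copy against a direction to the offset component pivot, join a reflection-textured rim.
# Particle color tints the sum before MF_FX_EmissiveAdjustments.
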Material /BRCosmetics/Characters/Player/Male/Medium/Bodies/M_MED_Alien_Agent/FX/M_Alien_Agent_Head_Bubbles
graph {
 properties {bUsedWithNiagaraSprites=true, blendMode=BLEND_Additive, pixelDepthOffsetMode=PDOM_Legacy, shadingModel=MSM_Unlit, usedShadingModels=Unlit}
 nodes [
  n1 = ParticlePositionWS()
  n2 = Constant3Vector() {constant=<rgba 0 0 -3 0>}
  Add(A=@n1, B=@n2)
  n3 = ComponentMask(None=TransformPosition(None=@n1) {transformSourceType=TRANSFORMPOSSOURCE_World}) {r=true}
  n4 = Saturate(None=Divide(A=Abs(None=@n3), B=ScalarParameter() {defaultValue=6, parameterName=Distance Squish}))
  n5 = MaterialFunctionCall!(UV, Size, Curvature) {materialFunction=/Game/Effects/Fort_Effects/MaterialFunctions/MF_Sphere}
  DepthFade(Opacity=Multiply(A=@n4, B=@n5.Mask), FadeDistance=ScalarParameter() {defaultValue=6, parameterName=Depth Fade Distance})
  n6 = Add(
   A=MaterialFunctionCall() {materialFunction=&ComponentPivotLocation}.Result
   B=@n2
  )
  n7 = Transform(None=Normalize(VectorInput=Subtract(A=@n6, B=@n1))) {transformSourceType=TRANSFORMSOURCE_World, transformType=TRANSFORM_View}
  n8 = ConstantBiasScale(None=DotProduct(A=Transform(None=@n5.Normal) {transformType=TRANSFORM_View}, B=@n7))
  n9 = ScalarParameter() {defaultValue=33, parameterName=Fake Specular Exponent, sliderMax=333}
  n10 = Multiply(A=@n5.Normal, B=Constant3Vector() {constant=<rgba -1 -1 1 0>}) {constB=-1}
  n11 = ConstantBiasScale(None=DotProduct(A=Transform(None=@n10) {transformType=TRANSFORM_View}, B=@n7))
  n12 = Add(A=Power(Base=OneMinus(None=@n8), Exp=@n9), B=Power(Base=OneMinus(None=@n11), Exp=@n9))
  n13 = ComponentMask(None=@n5.Normal) {g=true, r=true}
  n14 = Multiply(
   A=Power(
    Base=MaterialFunctionCall(Vector 2=@n13) {materialFunction=/Engine/Functions/Engine_MaterialFunctions02/Utility/VectorLength}.V2 Length
    Exp=ScalarParameter() {defaultValue=6, parameterName=Rim Exponent}
   )
   B=TextureSample() {texture=/Game/Packages/Fortress_SharedMaterials/Textures/ReflectionMap2}.RGB
  )
  MaterialFunctionCall(Emissive=Multiply(A=Add(A=@n12, B=@n14), B=ParticleColor().RGB)) {materialFunction=&MF_FX_EmissiveAdjustments}
  n15 = Multiply(A=@n4, B=@n4)
  n16 = Multiply(A=@n15, B=@n15)
  Multiply(A=Subtract(A=ParticlePositionWS(), B=WorldPosition().XYZ), B=@n16)
  OneMinus(None=@n16)
  Subtract(A=@n1)
  VectorParameter() {desc=Replaced with particle color because of new variant, parameterName=Hightlight Color, defaultValue=<rgba 1 5 0 0>}
 ]
}
