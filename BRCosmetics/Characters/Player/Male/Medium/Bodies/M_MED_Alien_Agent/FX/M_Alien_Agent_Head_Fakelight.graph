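# Additive unlit Niagara particle material that fakes a local light glow for the Alien Agent head cosmetic.
# Particle RGB, with a mobile-only boost, feeds the emissive adjustments function.
# A radial mask compares the world position behind translucency to the particle position, using radius and edge falloff, weighted toward downward vertex normals and raised to particle alpha.
# Scene world normals modulate the distance fade except on mobile; both color parameters are unconnected leftovers.
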Material /BRCosmetics/Characters/Player/Male/Medium/Bodies/M_MED_Alien_Agent/FX/M_Alien_Agent_Head_Fakelight
graph {
 properties {
  bUsedWithNiagaraMeshParticles=true
  bUsedWithNiagaraSprites=true
  blendMode=BLEND_Additive
  pixelDepthOffsetMode=PDOM_Legacy
  shadingModel=MSM_Unlit
  translucencyPass=MTP_BeforeDOF
  usedShadingModels=Unlit
 }
 nodes [
  n1 = Constant() {r=1}
  n2 = ScalarParameter() {defaultValue=2, parameterName=Mobile Color Boost}
  n3 = ParticleColor()
  MaterialFunctionCall(Emissive=Multiply(A=ShadingPathSwitch(Default=@n1, Mobile=@n2) {desc=Migrated from FeatureLevelSwitch}, B=@n3.RGB)) {materialFunction=&MF_FX_EmissiveAdjustments}
  n4 = ConstantBiasScale(None=ComponentMask(None=VertexNormalWS()) {b=true})
  n5 = ComponentMask(None=SceneTexture() {sceneTextureId=PPI_WorldNormal}.Color) {b=true}
  n6 = OneMinus(None=Saturate(None=Multiply(A=@n5) {constB=-1}))
  n7 = MaterialFunctionCall() {materialFunction=&WorldPositionBehindTranslucency}
  n8 = ParticlePositionWS()
  n9 = Saturate(None=Divide(A=Distance(A=@n7.Result, B=@n8), B=ScalarParameter() {defaultValue=8.5, parameterName=Radius}))
  n10 = ShadingPathSwitch(Default=OneMinus(None=Multiply(A=@n6, B=@n9)), Mobile=OneMinus(None=@n9)) {desc=Migrated from FeatureLevelSwitch}
  n11 = Saturate(None=Divide(A=OneMinus(None=@n9), B=ScalarParameter() {defaultValue=0.1, parameterName=Falloff}))
  Power(Base=Multiply(A=OneMinus(None=@n4), B=Multiply(A=@n10, B=@n11)), Exp=@n3.A)
  VectorParameter() {desc=Replaced with particle color because of new variant, parameterName=Color Mobile, defaultValue=<rgba 0.2 1 0 1>}
  VectorParameter() {parameterName=Color, defaultValue=<rgba 0.1 0.5 0 1>}
 ]
}
