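# Niagara head effect for the Alien Agent character body, combining rising bubbles with a green or orange glow selected by User.Variant.
# Bubbles spawn continuously in a flat cylinder, drift within a sphere blocking volume, and use a distance-kill module.
# Persistent burst particles drive a real light and a fake-light mesh positioned from the user skeletal mesh.
# User.bisHeadHidden drives visibility tags on bubbles and mesh, and disables the light.
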NiagaraSystem /BRCosmetics/Characters/Player/Male/Medium/Bodies/M_MED_Alien_Agent/FX/NS_Alien_Agent_Head
dependencies {
 usedDataInterfaces [&NiagaraDataInterfaceArrayColor, &NiagaraDataInterfaceCurve]
 usedDynamicInputs []
 usedModules [&SystemState]
}
emitters {
 Bubbles {
  emitterUpdateScript [
   &EmitterState()
   &SpawnRate(SpawnRate=11, Use Spawn Probability=false)
  ]
  particleSpawnScript [
   &InitializeParticle(
    Color=&SelectLinearColorFromArray(
     Array Sampling Mode=::NewEnumerator1
     Direct Array Index=&MakeIntFromFloat(Float=$User.Variant, Conversion Mode=::NewEnumerator1)
     Color Selection Array=DataInterface() {propertyValues {ColorData [<rgba 1 5 0 1>, <rgba 5 2.75 0 1>]}}
     Random Seed=?
    )
    Lifetime=5
    Sprite Rotation Mode=::NewEnumerator2
    Sprite Size Mode=::NewEnumerator1
    Uniform Sprite Size Max=2
    Uniform Sprite Size Min=1
    UsePositionOffset=false
    Write Material Random=true
   )
   &CylinderLocation(
    Cylinder Axis:Vector3f=<1 0.5 0>
    Cylinder Height=0
    Cylinder Radius=3
    MaskSpawnGroup=false
    Offset=<-3 0 0>
    Offset Mode:ENiagara_OffsetMode=::NewEnumerator0
    Override Seed=false
    Rotation Mode:ENiagara_RotationMode=::NewEnumerator1
    UseOwnerScale=true
   )
  ]
  particleUpdateScript [
   &ParticleState()
   &self:SphereBlockingVolume(Pivot Offset:Vector3f=<15 4 0>, Sphere Origin:Vector3f=/Niagara/DynamicInputs/Helpers/SimulationPosition(), Sphere Radius=11) {module=SphereBlockingVolume}
   &AccelerationForce(Acceleration=&RandomRangeVector(Minimum=<3 2.5 2>, Maximum=<9 -1 -2>), Coordinate Space=::Local)
   &SolveForcesAndVelocity(Clamp Velocity=false, Limit Acceleration=false)
   /Game/Effects/Niagara/Emitters/Water/FN_DistanceKillParticles(
    AgeReduiceAmount=0.3
    CenterPosition:Vector3f=&MakeVector(
     X=&MakeFloatFromVector(Vector=&SimulationPosition(), Channel=::NewEnumerator0)
     Y=&MakeFloatFromVector(Vector=&SimulationPosition(), Channel=::NewEnumerator1)
     Z=&MakeFloatFromVector(Vector=Linked() {linkedVariable {name=Particles.Position, type=NiagaraPosition}}, Channel=::NewEnumerator2)
    )
   )
   &SpriteRotationRate(Rotation Rate=&RandomRangeFloat(Minimum=90, Maximum=-90))
   &ScaleSpriteSize(Uniform Curve Sprite Scale=DataInterface() {propertyValues {Curve {keys [
        {arriveTangent=5.629, interpMode=RCIM_Cubic, leaveTangent=5.629, tangentMode=RCTM_User, time=0, value=0}
        {tangentMode=RCTM_User, time=0.5, value=1}
       ]}}})
   /SetVariables(Particles.VisibilityTag=&MakeCustomIntFromBool(Bool=$User.bisHeadHidden, True INT=1, False INT=0)) {isSetParameters=true, module=SetVariables_#1}
  ]
  rendererClasses [&NiagaraSpriteRendererProperties]
  renderers [{rendererClass=&NiagaraSpriteRendererProperties}]
 }
 Fake_Light {
  emitterUpdateScript [
   &EmitterState(
    Inactive Response=::NewEnumerator2
    Life Cycle Mode=::NewEnumerator1
    Loop Behavior=::NewEnumerator1
    Loop Duration=1.3826
    UseLoopDelay=false
   )
   &SpawnBurst_Instantaneous(Use Spawn Probability=false, UseLoopCountLimit=false)
  ]
  particleSpawnScript [&InitializeParticle(
    Color=&SelectLinearColorFromArray(
     Array Sampling Mode=::NewEnumerator1
     Direct Array Index=&MakeIntFromFloat(Float=$User.Variant, Conversion Mode=::NewEnumerator1)
     Color Selection Array=DataInterface() {propertyValues {ColorData [<rgba 0.1 0.5 0 1>, <rgba 0.5 0.275 0 1>]}}
     Random Seed=?
    )
    Position Offset=<4 2 0>
    UsePositionOffset=true
    Write Material Random=true
   )]
  particleUpdateScript [
   &ParticleState(Kill Particles When Lifetime Has Elapsed=false)
   &SkeletalMeshLocation(Bone Sampling Mode=::NewEnumerator4, MaskSpawnGroup=false, Skeletal Mesh=$User.Skeletal Mesh)
   /SetVariables(Particles.VisibilityTag=&MakeCustomIntFromBool(Bool=$User.bisHeadHidden, True INT=1, False INT=0)) {isSetParameters=true, module=SetVariables_#1}
  ]
  rendererClasses [&NiagaraMeshRendererProperties]
  renderers [{rendererClass=&NiagaraMeshRendererProperties, rendererIndex=0}]
 }
 Light {
  emitterUpdateScript [
   &EmitterState(Inactive Response=::NewEnumerator2, Life Cycle Mode=::NewEnumerator1, Loop Behavior=::NewEnumerator1, UseLoopDelay=false)
   &SpawnBurst_Instantaneous(Use Spawn Probability=false, UseLoopCountLimit=false)
  ]
  particleSpawnScript [&InitializeParticle(
    Color=&SelectLinearColorFromArray(
     Array Sampling Mode=::NewEnumerator1
     Direct Array Index=&MakeIntFromFloat(Float=$User.Variant, Conversion Mode=::NewEnumerator1)
     Color Selection Array=DataInterface() {propertyValues {ColorData [<rgba 10 50 0 1>, <rgba 50 27.5 0 1>]}}
     Random Seed=?
    )
    Position Offset=<2 2 0>
    UsePositionOffset=true
    Write Material Random=true
   )]
  particleUpdateScript [
   &ParticleState(Kill Particles When Lifetime Has Elapsed=false)
   &CameraOffset()
   /SetVariables(Particles.LightEnabled:NiagaraBool=&Boolean_NotOperation(Bool=$User.bisHeadHidden)) {isSetParameters=true}
  ]
  rendererClasses [&NiagaraLightRendererProperties]
  renderers [{rendererClass=&NiagaraLightRendererProperties, rendererIndex=0}]
 }
}
summary {
 systemName=NS_Alien_Agent_Head
 emitters [
  {emitterName=Light, rendererClasses [&NiagaraLightRendererProperties]}
  {emitterName=Bubbles, rendererClasses [&NiagaraSpriteRendererProperties]}
  {emitterName=Fake_Light, rendererClasses [&NiagaraMeshRendererProperties]}
 ]
 userVariables [
  User.bisHeadHidden {type=NiagaraBool}
  User.Skeletal Mesh {
   type=NiagaraDataInterfaceSkeletalMesh
   defaultValue {
    struct=NiagaraExt_VariableValue_DataInterface
    value {dataInterface=&self:NiagaraDataInterfaceSkeletalMesh_1}
   }
  }
  User.Variant {type=NiagaraFloat, defaultValue=0, description=""}
 ]
}
