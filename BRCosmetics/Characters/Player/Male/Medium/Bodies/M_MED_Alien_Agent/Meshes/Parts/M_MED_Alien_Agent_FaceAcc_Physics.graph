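# Physics asset for the Alien Agent male medium character's face accessory mesh, covering its brain bones, likely for secondary jiggle motion.
# A larger capsule body sits on Dyn_Brain and a small one on Dyn_Master_Brain.
# A single constraint joins the two, with narrow, equal swing limits and almost no twist, so the brain can only wobble slightly relative to its master bone.
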PhysicsAsset /BRCosmetics/Characters/Player/Male/Medium/Bodies/M_MED_Alien_Agent/Meshes/Parts/M_MED_Alien_Agent_FaceAcc_Physics
bodies {
 Dyn_Brain [{shapeType=Capsule, length=3, radius=3, shapeName=Dyn_Brain_capsule, center=<1 0 0>, rotation=<rot 90 0 0>}]
 Dyn_Master_Brain [{shapeType=Capsule, length=0.505, radius=0.505, shapeName=Dyn_Master_Brain_capsule, rotation=<rot 0 -90 0>}]
}
constraints [{bone1Name=Dyn_Brain, bone2Name=Dyn_Master_Brain, swing1LimitDegrees=15, swing2LimitDegrees=15, twistLimitDegrees=1}]
